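# Niagara system for a waterfall fountain effect on the Boiler prop.
# It groups six emitters in A and B pairs named for waterfall water, splashes and droplets.
# The only listed module is SystemState, and no user parameters are exposed.
# Emitter renderers and internal modules are not recorded, so the look of each layer is not shown.
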NiagaraSystem /Hermes/Props/Boiler/FX/NS_LWE_Boiler_WaterfallFountain
dependencies {usedDataInterfaces [], usedDynamicInputs [], usedModules [&SystemState]}
emitters {
 Water_Droplets_A {rendererClasses [], renderers []}
 Water_Droplets_B {rendererClasses [], renderers []}
 Water_Splashes_A {rendererClasses [], renderers []}
 Water_Splashes_B {rendererClasses [], renderers []}
 Water_Waterfall_A {rendererClasses [], renderers []}
 Water_Waterfall_B {rendererClasses [], renderers []}
}
summary {
 systemName=NS_LWE_Boiler_WaterfallFountain
 emitters [
  {emitterName=Water_Waterfall_A, rendererClasses []}
  {emitterName=Water_Waterfall_B, rendererClasses []}
  {emitterName=Water_Splashes_B, rendererClasses []}
  {emitterName=Water_Splashes_A, rendererClasses []}
  {emitterName=Water_Droplets_B, rendererClasses []}
  {emitterName=Water_Droplets_A, rendererClasses []}
 ]
 userVariables []
}
